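# MetaSound Source preset that plays random music stingers for the Jethro Outer Rim POI music encounter.
# It inherits its graph from MSS_MusicEncounter_RandomStingerFromArray and supplies an array of eight Outer Rim one-shot wave assets.
# Inputs expose master volume plus an encounter pause flag with a fade-out time.
# Output is stereo one-shot audio with a finished trigger, routed to the Fort_Music sound class.
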MetaSoundSource /JethroLager/Assets/Audio/POI/OuterRim/MSS_Jethro_OuterRim_MusicEncounter_RandomStingerFromArray_IP
preset_parent=/MusicEncounter/MetaSounds/MSS_MusicEncounter_RandomStingerFromArray
inputs [
 MasterVolume {type=Float}
 Music.WaveAsset {type=WaveAsset:Array}
 MusicEncounter.bIsPaused {type=Bool}
 MusicEncounter.PauseFadeOutTime {type=Time}
 UE.Source.OnPlay {type=Trigger}
]
interfaces [UE.OutputFormat.Stereo:1.0, UE.Source.OneShot:1.0, UE.Source:1.1]
outputs [
 "UE.OutputFormat.Stereo.Audio:0" {type=Audio}
 "UE.OutputFormat.Stereo.Audio:1" {type=Audio}
 UE.Source.OneShot.OnFinished {type=Trigger}
]
pages [Default {input_defaults {
   MasterVolume=1
   MusicEncounter.PauseFadeOutTime=3
   MusicEncounter.bIsPaused=null
   UE.Source.OnPlay=null
   Music.WaveAsset [
    /JethroLager/Assets/Audio/POI/AmbientOneShots/MX_S35_Jethro_Outer_Rim_01_ONE_SHOT
    /JethroLager/Assets/Audio/POI/AmbientOneShots/MX_S35_Jethro_Outer_Rim_02_ONE_SHOT
    /JethroLager/Assets/Audio/POI/AmbientOneShots/MX_S35_Jethro_Outer_Rim_03_ONE_SHOT
    /JethroLager/Assets/Audio/POI/AmbientOneShots/MX_S35_Jethro_Outer_Rim_04_ONE_SHOT
    /JethroLager/Assets/Audio/POI/AmbientOneShots/MX_S35_Jethro_Outer_Rim_05_ONE_SHOT
    /JethroLager/Assets/Audio/POI/AmbientOneShots/MX_S35_Jethro_Outer_Rim_06_ONE_SHOT
    /JethroLager/Assets/Audio/POI/AmbientOneShots/MX_S35_Jethro_Outer_Rim_07_ONE_SHOT
    /JethroLager/Assets/Audio/POI/AmbientOneShots/MX_S35_Jethro_Outer_Rim_08_ONE_SHOT
   ]
  }}]
properties {
 assetImportData=&self:AssetImportData
 numChannels=2
 outputFormat=Stereo
 sampleRate=48000
 soundClassObject=/Game/Sounds/FortSoundClassesAndModes/Fort_Music
}
